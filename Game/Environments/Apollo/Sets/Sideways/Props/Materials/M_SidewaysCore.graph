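# Clear Coat material for the Sideways core prop, combining rock, glossy crystal, dust and glowing emissive.
# A packed mask texture blends the layers and masks emissive, while actor-relative height gradients drive ground fades.
# Inverted Fresnel shapes the glow, which a world-aligned ramp pulses by scrolling over time.
# Shading path switches drop spot dirt, the pulse and scratched crystal normal detail on Forward and Mobile.
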Material /Game/Environments/Apollo/Sets/Sideways/Props/Materials/M_SidewaysCore
graph {
 properties {bUseMaterialAttributes=true, pixelDepthOffsetMode=PDOM_Legacy, shadingModel=MSM_ClearCoat, usedShadingModels=Clear Coat}
 nodes [
  n1 = ComponentMask(None=Subtract(A=WorldPosition().XYZ, B=ActorPositionWS())) {b=true}
  n2 = ScalarParameter() {defaultValue=355, group=Emissive, parameterName=GroundMask 1}
  NamedRerouteDeclaration(None=Saturate(None=Divide(A=@n1, B=@n2))) {name=GroundMask, nodeColor=<rgba 0 0.2266 1 1>}
  n3 = SetMaterialAttributes!(
   MaterialAttributes
   Base Color=VectorParameter() {group=Dirt, parameterName=Dust, defaultValue=<rgba 0.2052 0.1019 0.0438 0>}.RGB
   Specular=Constant() {r=0.5}
   Metallic=Constant()
   Roughness=LinearInterpolate(A=Constant() {r=-1}, B=Constant() {r=1.5}, Alpha=NamedRerouteUsage())
   Clear Coat=Constant()
   Clear Coat Roughness=Constant() {r=2}
  ) {attributeSetTypes [
    #1
    #2
    #3
    #4
    #5
    #6
   ]}
  NamedRerouteDeclaration(None=@n3) {name=Dirt, nodeColor=<rgba 0.7578 0 1 1>}
  NamedRerouteDeclaration(None=Divide(A=@n1, B=ScalarParameter() {defaultValue=750, group=Emissive, parameterName="GroundMask2 "})) {name=GroundMask_2, nodeColor=<rgba 0 0.2266 1 1>}
  n4 = TextureCoordinate()
  n5 = ScalarParameter() {defaultValue=8, group=Dirt, parameterName=Dirt Crystals}
  n6 = TextureSample(UVs=NamedRerouteDeclaration(None=Multiply(A=@n4, B=@n5)) {name=Dirt UV, nodeColor=<rgba 1 0.7266 0 1>}) {samplerType=SAMPLERTYPE_Masks, texture=/Game/Athena/WinterArt/Textures/T_DirtLeaks_M}
  NamedRerouteDeclaration(None=@n6.B) {name=Spots, nodeColor=<rgba 0 0.3437 1 1>}
  n7 = ScalarParameter() {defaultValue=200, group=Dirt, parameterName=DirtMask}
  NamedRerouteDeclaration(None=Saturate(None=Divide(A=@n1, B=@n7))) {name=DirtMask, nodeColor=<rgba 0 0.2266 1 1>}
  n8 = NamedRerouteUsage()
  n9 = Constant() {r=0.1876}
  n10 = NamedRerouteDeclaration(None=@n6.R) {name=Dirt_1, nodeColor=<rgba 0 0.3906 1 1>}
  n11 = Constant()
  n12 = TextureSampleParameter2D() {
   group=Base
   parameterName=Mask
   samplerType=SAMPLERTYPE_Masks
   texture=/Game/Environments/Apollo/Sets/Sideways/Props/Textures/T_Sideways_Core_01_CM1
  }
  n13 = SetMaterialAttributes!(
   MaterialAttributes
   Base Color=MaterialFunctionCall(Base=VectorParameter() {group=Base, parameterName=Rock, defaultValue=<rgba 0.1667 0.1667 0.1667 0>}.RGB, Blend=@n12.G) {materialFunction=&Blend_Overlay}.Result
   Metallic=Constant()
   Specular=Constant() {r=0.2}
   Roughness=Constant() {r=0.65}
   Clear Coat=Constant()
   Normal=TextureSample(UVs=NamedRerouteUsage()) {samplerType=SAMPLERTYPE_Normal, texture=/Game/Environments/Apollo/Rocks/Natural_Rocks/Textures/T_RockDetails_02_N}.RGB
  ) {attributeSetTypes [
    #1
    #3
    #2
    #4
    #5
    #7
   ]}
  n14 = SetMaterialAttributes!(
   MaterialAttributes
   Base Color=VectorParameter() {group=Base, parameterName=Crystal, defaultValue=<rgba 0.0334 0 0.0469 0>}.RGB
   Specular=ScalarParameter() {defaultValue=0.1, group=Base, parameterName=Crystal S}
   Metallic=ScalarParameter() {defaultValue=0.5, group=Base, parameterName=Crystal M}
   Roughness=ScalarParameter() {defaultValue=0.35, group=Base, parameterName=Crystal R}
   Clear Coat=Constant() {r=1}
   Clear Coat Roughness=ScalarParameter() {group=Base, parameterName=Crystal Clearcoat R}
  ) {attributeSetTypes [
    #1
    #2
    #3
    #4
    #5
    #6
   ]}
  n15 = TextureSampleParameter2D() {
   group=Base
   parameterName=Mask
   samplerType=SAMPLERTYPE_Masks
   texture=&T_Sideways_Core_01_CM1
  }
  n16 = BlendMaterialAttributes(A=@n13, B=@n14, Alpha=@n15.A)
  n17 = BlendMaterialAttributes(
   A=@n16
   B=NamedRerouteUsage()
   Alpha=Multiply(
    A=ScalarParameter() {defaultValue=0.25, group=Dirt, parameterName=SpotsDirt}
    B=MaterialFunctionCall(Transition Phase=@n9, Height Texture=@n10, Contrast=@n11) {materialFunction=&HeightLerp}.Lerp Alpha No Contrast
   )
  )
  n18 = ShadingPathSwitch(Default=@n17, Deferred=@n17, Forward=@n16, Mobile=@n16)
  n19 = OneMinus(None=ScalarParameter() {defaultValue=1, group=Dirt, parameterName=Dust Value})
  n20 = TextureSampleParameter2D() {
   group=Base
   parameterName=Mask
   samplerType=SAMPLERTYPE_Masks
   texture=&T_Sideways_Core_01_CM1
  }
  n21 = Multiply(A=Max(A=NamedRerouteUsage(), B=Saturate(None=@n19)), B=LinearInterpolate(Alpha=@n20.R) {constA=0.5})
  n22 = NamedRerouteUsage()
  n23 = LinearInterpolate(
   A=VectorParameter() {group=Emissive, parameterName=EmissiveColor2, defaultValue=<rgba 2 0 0.4187 0>}.RGB
   B=VectorParameter() {group=Emissive, parameterName=EmissiveTop, defaultValue=<rgba 1 0.0927 0 0>}.RGB
   Alpha=NamedRerouteUsage()
  )
  n24 = LinearInterpolate(A=Constant() {r=0.35}, B=Constant() {r=2}, Alpha=NamedRerouteUsage())
  n25 = TextureSampleParameter2D() {
   group=Base
   parameterName=Mask
   samplerType=SAMPLERTYPE_Masks
   texture=&T_Sideways_Core_01_CM1
  }
  n26 = TextureSampleParameter2D() {
   group=Emissive
   parameterName=Emissive MAsk
   samplerType=SAMPLERTYPE_LinearGrayscale
   texture=/Game/Environments/Apollo/Sets/Sideways/Props/Textures/T_Sideways_Core_01_CM2
  }
  n27 = Multiply(
   A=LinearInterpolate(
    A=Multiply(A=@n23, B=@n24)
    B=VectorParameter() {group=Emissive, parameterName=EmissiveColor, defaultValue=<rgba 0.1594 0.0088 0.0583 0>}.RGB
    Alpha=@n25.R
   )
   B=@n26.R
  )
  n28 = ScalarParameter() {defaultValue=2, group=Emissive, parameterName=Fresnel Exponent}
  n29 = ScalarParameter() {group=Emissive, parameterName=Fresnel Frac}
  n30 = Multiply(
   A=Multiply(A=ScalarParameter() {defaultValue=0.35, group=Emissive, parameterName=EmissiveValue}, B=Multiply(A=@n27, B=@n25.B))
   B=OneMinus(None=Fresnel(ExponentIn=@n28, BaseReflectFractionIn=@n29))
  )
  n31 = GetMaterialAttributes(None=SetMaterialAttributes!(MaterialAttributes=BlendMaterialAttributes(A=@n8, B=@n18, Alpha=@n21), Emissive Color=Multiply(A=@n22, B=@n30)) {attributeSetTypes [#8]}) {attributeGetTypes [#8]}
  n32 = Constant() {r=0.75}
  n33 = WorldPosition()
  n34 = ActorPositionWS()
  n35 = Time()
  n36 = Constant3Vector() {constant=<rgba 0 0 -200 0>}
  n37 = MaterialFunctionCall(
   TextureObject=TextureObject() {samplerType=SAMPLERTYPE_Masks, texture=/Game/Environments/Sets/Arid/Textures/T_TimeRamp}
   TextureSize=Constant() {r=1555}
   WorldPosition=Add(A=Subtract(A=@n33.XYZ, B=@n34), B=Multiply(A=@n35, B=@n36))
  ) {materialFunction=&WorldAlignedTexture}
  n38 = SetMaterialAttributes(MaterialAttributes=@n31.MaterialAttributes, Emissive Color=Multiply(A=@n31.Emissive Color, B=Max(A=@n32, B=@n37.XY Texture))) {attributeSetTypes [#8]}
  n39 = GetMaterialAttributes(None=ShadingPathSwitch(Default=@n38, Deferred=@n38, Forward=@n31.MaterialAttributes, Mobile=@n31.MaterialAttributes)) {attributeGetTypes [#7]}
  n40 = Multiply(A=@n39.Normal, B=Constant3Vector() {constant=<rgba 1 1 0 0>})
  n41 = TextureSampleParameter2D() {
   group=Base
   parameterName=Normals
   samplerType=SAMPLERTYPE_Normal
   texture=/Game/Environments/Apollo/Sets/Sideways/Props/Textures/T_Sideways_Core_01_N
  }
  n42 = Multiply(A=TextureCoordinate(), B=ScalarParameter() {defaultValue=3, group=Base, parameterName=NormalDetails Scale})
  n43 = Multiply(
   A=TextureSample(UVs=@n42) {samplerType=SAMPLERTYPE_Normal, texture=/Game/Global/Textures/Default/Tileable_Patterns/T_IceScratch1_N}.RGB
   B=Constant3Vector() {constant=<rgba 1 1 0 0>}
  )
  n44 = Constant() {r=-0.2}
  n45 = MaterialFunctionCall(
   A=@n41.RGB
   B=Add(A=@n41.RGB, B=Multiply(A=@n43, B=@n44))
   Transition Phase=Constant() {r=0.401}
   Height Texture=TextureSample(UVs=@n42) {samplerType=SAMPLERTYPE_Masks, texture=/Game/Global/Textures/Default/Tileable_Patterns/T_IceScratch1_M}.R
   Contrast=Constant() {r=12}
  ) {materialFunction=&HeightLerp}
  n46 = TextureSampleParameter2D() {
   group=Base
   parameterName=Mask
   samplerType=SAMPLERTYPE_Masks
   texture=&T_Sideways_Core_01_CM1
  }
  SetMaterialAttributes!(
   MaterialAttributes=@n39.MaterialAttributes
   Normal=LinearInterpolate(A=Add(A=@n40, B=@n41.RGB), B=ShadingPathSwitch(Default=@n45.Results, Deferred=@n45.Results, Forward=@n41.RGB, Mobile=@n41.RGB), Alpha=@n46.A)
  ) {attributeSetTypes [#7]}
 ]
}
